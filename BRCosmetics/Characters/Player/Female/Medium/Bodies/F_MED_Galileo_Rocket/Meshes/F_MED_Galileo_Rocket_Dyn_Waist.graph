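# Physics asset for the dynamic waist cloth bones on the Galileo Rocket medium female player body.
# Four loincloth chains (front/back, left/right) of boxes and capsules use limited swing and twist, hanging from tiny master bodies free-constrained to the pelvis.
# A free constraint also links the two left chain tips.
# Pelvis, spine, leg, right-thigh gun and huge offset root capsules provide body shapes, the root likely serving as a floor.
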PhysicsAsset /BRCosmetics/Characters/Player/Female/Medium/Bodies/F_MED_Galileo_Rocket/Meshes/F_MED_Galileo_Rocket_Dyn_Waist
bodies {
 calf_l [{shapeType=Capsule, length=32.5625, radius=5.625, center=<-18.4619 -4.0767 -0.475>, rotation=<rot 87.4746 -178.8815 -176.291>}]
 calf_r [{shapeType=Capsule, length=32.5625, radius=5.625, center=<18.4619 4.0767 0.475>, rotation=<rot 87.4746 -178.8815 -176.291>}]
 dyn_gun [{shapeType=Capsule, length=19.7188, radius=5.7812, center=<-6.4037 0.2954 -0.0722>, rotation=<rot 88.1626 163.6002 164.5377>}]
 dyn_loincloth_bk_l_01 [{shapeType=Box, extentX=10.375, extentY=12.8125, extentZ=1.0625, center=<4.8598 0 -0.0001>, rotation=<rot 0 0 -4.3162>}]
 dyn_loincloth_bk_l_02 [{shapeType=Capsule, length=4.25, radius=5.75, shapeName=dyn_loincloth_bk_l_02_capsule, center=<5.026 0.0001 3.2>, rotation=<rot 90 0 0>}]
 dyn_loincloth_bk_l_03 [{shapeType=Box, extentX=11.5, extentY=11.4375, extentZ=3.625, center=<5.8018 0.2633 1.2>, rotation=<rot 0 0 11.9985>}]
 dyn_loincloth_bk_l_04 [{shapeType=Capsule, length=4, radius=5, shapeName=dyn_loincloth_bk_l_04_capsule, center=<3.2321 0 3>, rotation=<rot 90 0 0>}]
 dyn_loincloth_bk_l_05 [{shapeType=Box, extentX=11, extentY=7.5, extentZ=3.8125, center=<4.0908 0.3967 0>, rotation=<rot -7.0924 -3.3102 25.0933>}]
 dyn_loincloth_bk_l_MSTR [{shapeType=Capsule, length=0.505, radius=0.505, rotation=<rot 0 -90 0>}]
 dyn_loincloth_bk_r_01 [{shapeType=Capsule, length=9, radius=4.75, shapeName=dyn_loincloth_bk_r_01_capsule, center=<2.2 0 4>, rotation=<rot 0 0 99.0848>}]
 dyn_loincloth_bk_r_02 [{shapeType=Box, extentX=11.125, extentY=13, extentZ=2.1875, center=<5.915 0 0>, rotation=<rot 0 0 7.7238>}]
 dyn_loincloth_bk_r_03 [{shapeType=Capsule, length=7, radius=8.5, shapeName=dyn_loincloth_bk_r_03_capsule, center=<7 0 7>, rotation=<rot 87.9736 89.9959 -179.9995>}]
 dyn_loincloth_bk_r_04 [{shapeType=Box, extentX=15.125, extentY=13, extentZ=2.1875, center=<6.9579 0 0>, rotation=<rot 0 0 7.7238>}]
 dyn_loincloth_bk_r_05 [{shapeType=Capsule, length=8.5, radius=5.5, shapeName=dyn_loincloth_fr_r_05_capsule, center=<5.2188 0 5.3>, rotation=<rot 0 90 90>}]
 dyn_loincloth_bk_r_06 [{shapeType=Capsule, length=8.5, radius=5.5, shapeName=dyn_loincloth_fr_r_05_capsule, center=<5.2188 0 0>, rotation=<rot 90 153.435 -116.565>}]
 dyn_loincloth_bk_r_07 [{
   shapeType=Capsule
   length=8.75
   radius=5.25
   shapeName=dyn_loincloth_fr_r_05_capsule
   center=<5.2188 0 0>
   rotation=<rot 90 153.435 -116.565>
  }]
 dyn_loincloth_bk_r_MSTR [{shapeType=Capsule, length=0.505, radius=0.505, rotation=<rot 0 -90 0>}]
 dyn_loincloth_fr_l_01 [{shapeType=Capsule, length=2.5, radius=3, shapeName=dyn_loincloth_fr_l_01_capsule, center=<2.4576 -0.0006 1.5>, rotation=<rot 90 0 90>}]
 dyn_loincloth_fr_l_02 [{shapeType=Box, extentX=7.875, extentY=7.5, extentZ=3, center=<3.2143 0 0.3805>, rotation=<rot 6.2814 0 0>}]
 dyn_loincloth_fr_l_03 [{shapeType=Capsule, length=1.5, radius=4.25, shapeName=dyn_loincloth_fr_l_03_capsule, center=<1.3268 -0.0007 2.2>, rotation=<rot 90 0 90>}]
 dyn_loincloth_fr_l_04 [{shapeType=Box, extentX=7.375, extentY=7.5, extentZ=4.375, center=<4.0028 0 -0.0001>}]
 dyn_loincloth_fr_l_05 [{
   shapeType=Capsule
   length=1.5
   radius=5.25
   shapeName=dyn_loincloth_fr_l_05_capsule
   center=<3.7372 -0.1413 2>
   rotation=<rot 76.9565 -89.9968 0.0001>
  }]
 dyn_loincloth_fr_l_06 [{
   shapeType=Capsule
   length=2.75
   radius=5.25
   shapeName=dyn_loincloth_fr_l_06_capsule
   center=<3.3641 -0.1523 2.5>
   rotation=<rot 0 0 74.6344>
  }]
 dyn_loincloth_fr_l_07 [{shapeType=Box, extentX=11.4375, extentY=7.5, extentZ=4.5625, center=<5.4137 -0.4332 0>, rotation=<rot -0.6365 0.234 -20.1926>}]
 dyn_loincloth_fr_l_MSTR [{shapeType=Capsule, length=0.505, radius=0.505, rotation=<rot 0 -90 0>}]
 dyn_loincloth_fr_r_01 [{shapeType=Box, extentX=11.75, extentY=13.5, extentZ=1.0625, center=<5.5169 0 0>}]
 dyn_loincloth_fr_r_02 [{shapeType=Box, extentX=16.25, extentY=13.5, extentZ=4.2251, center=<8.2997 0 0>}]
 dyn_loincloth_fr_r_03 [{shapeType=Box, extentX=16.625, extentY=13.5, extentZ=3.7251, center=<8.5007 0 0>}]
 dyn_loincloth_fr_r_04 [{shapeType=Box, extentX=13.375, extentY=13.5, extentZ=2.9375, center=<6.7214 0 0>}]
 dyn_loincloth_fr_r_05 [{
   shapeType=Capsule
   length=8.75
   radius=5.25
   shapeName=dyn_loincloth_fr_r_05_capsule
   center=<5.2188 0 0>
   rotation=<rot 90 153.435 -116.565>
  }]
 dyn_loincloth_fr_r_06 [{
   shapeType=Capsule
   length=8.75
   radius=5.25
   shapeName=dyn_loincloth_fr_r_05_capsule
   center=<5.2188 0 0>
   rotation=<rot 90 153.435 -116.565>
  }]
 dyn_loincloth_fr_r_MSTR [{shapeType=Capsule, length=0.505, radius=0.505, rotation=<rot 0 -90 0>}]
 pelvis [{shapeType=Capsule, length=9.6875, radius=10.9375, center=<-3.3306 -1.5292 1.0034>}]
 root [{shapeType=Capsule, length=0, radius=300, shapeName=root_capsule, center=<0 -7.915 -291.2693>, rotation=<rot 90 180 -90>}]
 spine_01 [{shapeType=Capsule, length=0.5625, radius=11.4688, center=<-3.3306 -1.5292 0>}]
 thigh_l [
  {shapeType=Capsule, length=34.875, radius=8.75, center=<-25.4575 -2.1367 0.6036>, rotation=<rot 84.1996 1.1194 0>}
  {shapeType=Capsule, length=29.4688, radius=15, center=<-27.5937 0.0004 13.0545>, rotation=<rot 90 0 0>}
 ]
 thigh_r [
  {shapeType=Capsule, length=29.1562, radius=9.25, center=<23.195 -0.3581 -1.4882>, rotation=<rot 88.2573 1.1193 -0.0001>}
  {shapeType=Capsule, length=28.5938, radius=7.6562, center=<20.4275 5.1388 -0.0004>, rotation=<rot 90 13.332 13.8214>}
 ]
}
constraints [
 {bone1Name=dyn_loincloth_bk_l_01, bone2Name=dyn_loincloth_bk_l_MSTR, swing1LimitDegrees=60, swing2LimitDegrees=60, twistLimitDegrees=25}
 {bone1Name=dyn_loincloth_bk_l_02, bone2Name=dyn_loincloth_bk_l_01, swing1LimitDegrees=60, swing2LimitDegrees=60, twistLimitDegrees=25}
 {bone1Name=dyn_loincloth_bk_l_03, bone2Name=dyn_loincloth_bk_l_02, swing1LimitDegrees=60, swing2LimitDegrees=60, twistLimitDegrees=25}
 {bone1Name=dyn_loincloth_bk_l_04, bone2Name=dyn_loincloth_bk_l_03, swing1LimitDegrees=60, swing2LimitDegrees=60, twistLimitDegrees=25}
 {bone1Name=dyn_loincloth_bk_l_05, bone2Name=dyn_loincloth_bk_l_04, swing1LimitDegrees=60, swing2LimitDegrees=60, twistLimitDegrees=25}
 {bone1Name=dyn_loincloth_bk_r_01, bone2Name=dyn_loincloth_bk_r_MSTR, swing1LimitDegrees=40, swing2LimitDegrees=40, twistLimitDegrees=25}
 {bone1Name=dyn_loincloth_bk_r_02, bone2Name=dyn_loincloth_bk_r_01, swing1LimitDegrees=40, swing2LimitDegrees=40, twistLimitDegrees=25}
 {bone1Name=dyn_loincloth_bk_r_03, bone2Name=dyn_loincloth_bk_r_02, swing1LimitDegrees=40, swing2LimitDegrees=40, twistLimitDegrees=25}
 {bone1Name=dyn_loincloth_bk_r_04, bone2Name=dyn_loincloth_bk_r_03, swing1LimitDegrees=40, swing2LimitDegrees=40, twistLimitDegrees=25}
 {bone1Name=dyn_loincloth_bk_r_05, bone2Name=dyn_loincloth_bk_r_04, swing1LimitDegrees=40, swing2LimitDegrees=40, twistLimitDegrees=25}
 {bone1Name=dyn_loincloth_bk_r_06, bone2Name=dyn_loincloth_bk_r_05, swing1LimitDegrees=40, swing2LimitDegrees=40, twistLimitDegrees=25}
 {bone1Name=dyn_loincloth_bk_r_07, bone2Name=dyn_loincloth_bk_r_06, swing1LimitDegrees=40, swing2LimitDegrees=40, twistLimitDegrees=25}
 {bone1Name=dyn_loincloth_fr_l_01, bone2Name=dyn_loincloth_fr_l_MSTR, swing1LimitDegrees=5, swing2LimitDegrees=5, twistLimitDegrees=5}
 {bone1Name=dyn_loincloth_fr_l_02, bone2Name=dyn_loincloth_fr_l_01, swing1LimitDegrees=45, swing2LimitDegrees=45, twistLimitDegrees=25}
 {bone1Name=dyn_loincloth_fr_l_03, bone2Name=dyn_loincloth_fr_l_02, swing1LimitDegrees=60, swing2LimitDegrees=60, twistLimitDegrees=25}
 {bone1Name=dyn_loincloth_fr_r_01, bone2Name=dyn_loincloth_fr_r_MSTR, swing1LimitDegrees=30, swing2LimitDegrees=30, twistLimitDegrees=5}
 {bone1Name=dyn_loincloth_fr_r_02, bone2Name=dyn_loincloth_fr_r_01, swing1LimitDegrees=40, swing2LimitDegrees=40, twistLimitDegrees=35}
 {bone1Name=dyn_loincloth_fr_r_03, bone2Name=dyn_loincloth_fr_r_02, swing1LimitDegrees=40, swing2LimitDegrees=40, twistLimitDegrees=25}
 {bone1Name=dyn_loincloth_fr_r_04, bone2Name=dyn_loincloth_fr_r_03, swing1LimitDegrees=40, swing2LimitDegrees=40, twistLimitDegrees=25}
 {bone1Name=dyn_loincloth_fr_r_05, bone2Name=dyn_loincloth_fr_r_04, swing1LimitDegrees=40, swing2LimitDegrees=40, twistLimitDegrees=25}
 {bone1Name=dyn_loincloth_fr_r_06, bone2Name=dyn_loincloth_fr_r_05, swing1LimitDegrees=40, swing2LimitDegrees=40, twistLimitDegrees=25}
 {bone1Name=dyn_loincloth_fr_l_04, bone2Name=dyn_loincloth_fr_l_03, swing1LimitDegrees=60, swing2LimitDegrees=60, twistLimitDegrees=25}
 {bone1Name=dyn_loincloth_fr_l_05, bone2Name=dyn_loincloth_fr_l_04, swing1LimitDegrees=60, swing2LimitDegrees=60, twistLimitDegrees=25}
 {bone1Name=dyn_loincloth_fr_l_06, bone2Name=dyn_loincloth_fr_l_05, swing1LimitDegrees=60, swing2LimitDegrees=60, twistLimitDegrees=25}
 {bone1Name=dyn_loincloth_fr_l_07, bone2Name=dyn_loincloth_fr_l_06, swing1LimitDegrees=60, swing2LimitDegrees=60, twistLimitDegrees=25}
 {
  bone1Name=dyn_loincloth_bk_l_MSTR
  bone2Name=pelvis
  swing1LimitDegrees=45
  swing1Motion=Free
  swing2LimitDegrees=45
  swing2Motion=Free
  twistLimitDegrees=45
  twistMotion=Free
 }
 {
  bone1Name=dyn_loincloth_bk_r_MSTR
  bone2Name=pelvis
  swing1LimitDegrees=45
  swing1Motion=Free
  swing2LimitDegrees=45
  swing2Motion=Free
  twistLimitDegrees=45
  twistMotion=Free
 }
 {
  bone1Name=dyn_loincloth_fr_l_MSTR
  bone2Name=pelvis
  swing1LimitDegrees=45
  swing1Motion=Free
  swing2LimitDegrees=45
  swing2Motion=Free
  twistLimitDegrees=45
  twistMotion=Free
 }
 {
  bone1Name=dyn_loincloth_fr_r_MSTR
  bone2Name=pelvis
  swing1LimitDegrees=45
  swing1Motion=Free
  swing2LimitDegrees=45
  swing2Motion=Free
  twistLimitDegrees=45
  twistMotion=Free
 }
 {
  bone1Name=thigh_l
  bone2Name=pelvis
  swing1LimitDegrees=45
  swing1Motion=Free
  swing2LimitDegrees=45
  swing2Motion=Free
  twistLimitDegrees=45
  twistMotion=Free
 }
 {
  bone1Name=thigh_r
  bone2Name=pelvis
  swing1LimitDegrees=45
  swing1Motion=Free
  swing2LimitDegrees=45
  swing2Motion=Free
  twistLimitDegrees=45
  twistMotion=Free
 }
 {
  bone1Name=dyn_loincloth_bk_l_05
  bone2Name=dyn_loincloth_fr_l_07
  swing1LimitDegrees=45
  swing1Motion=Free
  swing2LimitDegrees=45
  swing2Motion=Free
  twistLimitDegrees=45
  twistMotion=Free
 }
 {
  bone1Name=calf_r
  bone2Name=thigh_r
  swing1LimitDegrees=45
  swing1Motion=Free
  swing2LimitDegrees=45
  swing2Motion=Free
  twistLimitDegrees=45
  twistMotion=Free
 }
 {
  bone1Name=calf_l
  bone2Name=thigh_l
  swing1LimitDegrees=45
  swing1Motion=Free
  swing2LimitDegrees=45
  swing2Motion=Free
  twistLimitDegrees=45
  twistMotion=Free
 }
 {
  bone1Name=pelvis
  bone2Name=root
  swing1LimitDegrees=45
  swing1Motion=Free
  swing2LimitDegrees=45
  swing2Motion=Free
  twistLimitDegrees=45
  twistMotion=Free
 }
 {
  bone1Name=spine_01
  bone2Name=pelvis
  swing1LimitDegrees=45
  swing1Motion=Free
  swing2LimitDegrees=45
  swing2Motion=Free
  twistLimitDegrees=45
  twistMotion=Free
 }
 {
  bone1Name=dyn_gun
  bone2Name=thigh_r
  swing1LimitDegrees=45
  swing1Motion=Free
  swing2LimitDegrees=45
  swing2Motion=Free
  twistLimitDegrees=45
  twistMotion=Free
 }
]
