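# Physics asset for the dynamic hood on the female medium Galileo Rocket character body, built entirely from capsules.
# Head, spine, clavicle and upper-arm bodies form the torso structure around the hood, joined by free-swinging, free-twisting constraints.
# Left and right hood chains run from MSTR bodies on the clavicles through tight swing and zero-twist limits, each branching at 04 into 04a and 05.
# A free angular constraint links the two 05 tips, bridging both sides.
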PhysicsAsset /BRCosmetics/Characters/Player/Female/Medium/Bodies/F_MED_Galileo_Rocket/Meshes/F_MED_Galileo_Rocket_Dyn_hood
bodies {
 clavicle_l [{shapeType=Capsule, length=8.3125, radius=6.1875, center=<6.9162 -1.5082 -0.7981>, rotation=<rot 83.8126 0 2.4746>}]
 clavicle_r [{shapeType=Capsule, length=11.7812, radius=5.5312, center=<-6.8693 1.076 1.0818>, rotation=<rot 85.678 -1.0702 -3.5405>}]
 dyn_hood_l_01 [{shapeType=Capsule, length=5.5, radius=1.5, shapeName=dyn_hood_l_01_capsule, rotation=<rot 90 90 0>}]
 dyn_hood_l_02 [{shapeType=Capsule, length=4.75, radius=2.25, shapeName=dyn_hood_l_02_capsule, center=<1.8437 0.0001 1>, rotation=<rot 90 -14.0363 -104.0363>}]
 dyn_hood_l_02a [{shapeType=Capsule, length=7.5, radius=2.25, shapeName=dyn_hood_l_02a_capsule, center=<2.3285 0 1>, rotation=<rot 90 -90 180>}]
 dyn_hood_l_03 [{shapeType=Capsule, length=7.75, radius=2.5, shapeName=dyn_hood_l_03_capsule, center=<2.3 0 0.6>, rotation=<rot 90 14.0363 -75.9637>}]
 dyn_hood_l_04 [{
   shapeType=Capsule
   length=5.25
   radius=3.25
   shapeName=dyn_hood_l_04_capsule
   center=<2.3799 0.0001 0.8454>
   rotation=<rot 90 -14.0363 -104.0363>
  }]
 dyn_hood_l_04a [{shapeType=Capsule, length=3.5, radius=3.5, shapeName=dyn_hood_l_04a_capsule, center=<-0.6709 0 0>}]
 dyn_hood_l_05 [{shapeType=Capsule, length=2.25, radius=4.75, shapeName=dyn_hood_l_05_capsule, center=<2 0 2.7>, rotation=<rot 90 0 0>}]
 dyn_hood_l_MSTR [{shapeType=Capsule, length=0.505, radius=0.505, rotation=<rot 0 -90 0>}]
 dyn_hood_r_01 [{shapeType=Capsule, length=5.5, radius=1.5, shapeName=dyn_hood_r_01_capsule, rotation=<rot 90 90 0>}]
 dyn_hood_r_02 [{shapeType=Capsule, length=4.75, radius=2.25, shapeName=dyn_hood_r_02_capsule, center=<1.8437 0.0001 1>, rotation=<rot 90 -14.0363 -104.0363>}]
 dyn_hood_r_02a [{shapeType=Capsule, length=7.5, radius=2.25, shapeName=dyn_hood_r_02a_capsule, center=<2.3285 0 1>, rotation=<rot 90 -90 180>}]
 dyn_hood_r_03 [{shapeType=Capsule, length=7.25, radius=2.5, shapeName=dyn_hood_r_03_capsule, center=<2.3285 -0.0001 1>, rotation=<rot 90 -90 180>}]
 dyn_hood_r_04 [{shapeType=Capsule, length=3.5, radius=3.5, shapeName=dyn_hood_r_04_capsule, center=<2.3762 0.0001 2.4>, rotation=<rot 90 -90 180>}]
 dyn_hood_r_04a [{shapeType=Capsule, length=3.5, radius=3.5, shapeName=dyn_hood_r_04a_capsule, center=<-0.6709 0 0>}]
 dyn_hood_r_05 [{shapeType=Capsule, length=3.25, radius=3.75, shapeName=dyn_hood_r_05_capsule, center=<2 0 2.7>, rotation=<rot 90 0 0>}]
 dyn_hood_r_MSTR [{shapeType=Capsule, length=0.505, radius=0.505, rotation=<rot 0 -90 0>}]
 head [{shapeType=Capsule, length=0, radius=16.25, center=<8.4 -2.2452 0>, rotation=<rot 90 0 0>}]
 spine_02 [{shapeType=Capsule, length=3.4062, radius=9.7188, center=<5.1511 1.6991 0>}]
 spine_03 [{shapeType=Capsule, length=3.4062, radius=9.7188, center=<3.6584 1.5079 0.0001>}]
 spine_04 [{shapeType=Capsule, length=6.125, radius=8.7188, center=<3.6584 0.0494 0>}]
 spine_05 [{shapeType=Capsule, length=8.7188, radius=8.375, center=<2.6662 0.4658 -0.0261>, rotation=<rot 1.0982 0.0828 4.3147>}]
 upperarm_l [{shapeType=Capsule, length=18.25, radius=5, shapeName=upperarm_l_capsule, center=<12 0 0>, rotation=<rot 90 0 0>}]
 upperarm_r [{shapeType=Capsule, length=18.25, radius=5, shapeName=upperarm_r_capsule, center=<-12 0 0>, rotation=<rot 90 0 0>}]
}
constraints [
 {bone1Name=dyn_hood_l_01, bone2Name=dyn_hood_l_MSTR, swing1LimitDegrees=30, swing2LimitDegrees=20, twistLimitDegrees=0}
 {bone1Name=dyn_hood_l_02, bone2Name=dyn_hood_l_01, swing1LimitDegrees=30, swing2LimitDegrees=20, twistLimitDegrees=0}
 {bone1Name=dyn_hood_l_03, bone2Name=dyn_hood_l_02a, swing1LimitDegrees=30, swing2LimitDegrees=20, twistLimitDegrees=0}
 {bone1Name=dyn_hood_l_04, bone2Name=dyn_hood_l_03, swing1LimitDegrees=30, swing2LimitDegrees=20, twistLimitDegrees=0}
 {bone1Name=dyn_hood_l_05, bone2Name=dyn_hood_l_04, swing1LimitDegrees=30, swing2LimitDegrees=20, twistLimitDegrees=0}
 {bone1Name=dyn_hood_r_01, bone2Name=dyn_hood_r_MSTR, swing1LimitDegrees=30, swing2LimitDegrees=20, twistLimitDegrees=0}
 {bone1Name=dyn_hood_r_02, bone2Name=dyn_hood_r_01, swing1LimitDegrees=30, swing2LimitDegrees=20, twistLimitDegrees=0}
 {bone1Name=dyn_hood_r_03, bone2Name=dyn_hood_r_02a, swing1LimitDegrees=30, swing2LimitDegrees=20, twistLimitDegrees=0}
 {bone1Name=dyn_hood_r_04, bone2Name=dyn_hood_r_03, swing1LimitDegrees=30, swing2LimitDegrees=20, twistLimitDegrees=0}
 {bone1Name=dyn_hood_r_05, bone2Name=dyn_hood_r_04, swing1LimitDegrees=30, swing2LimitDegrees=20, twistLimitDegrees=0}
 {
  bone1Name=dyn_hood_l_MSTR
  bone2Name=clavicle_l
  swing1LimitDegrees=45
  swing1Motion=Free
  swing2LimitDegrees=45
  swing2Motion=Free
  twistLimitDegrees=45
  twistMotion=Free
 }
 {
  bone1Name=dyn_hood_r_MSTR
  bone2Name=clavicle_r
  swing1LimitDegrees=45
  swing1Motion=Free
  swing2LimitDegrees=45
  swing2Motion=Free
  twistLimitDegrees=45
  twistMotion=Free
 }
 {
  bone1Name=clavicle_l
  bone2Name=spine_05
  swing1LimitDegrees=45
  swing1Motion=Free
  swing2LimitDegrees=45
  swing2Motion=Free
  twistLimitDegrees=45
  twistMotion=Free
 }
 {
  bone1Name=clavicle_r
  bone2Name=spine_05
  swing1LimitDegrees=45
  swing1Motion=Free
  swing2LimitDegrees=45
  swing2Motion=Free
  twistLimitDegrees=45
  twistMotion=Free
 }
 {
  bone1Name=spine_05
  bone2Name=spine_04
  swing1LimitDegrees=45
  swing1Motion=Free
  swing2LimitDegrees=45
  swing2Motion=Free
  twistLimitDegrees=45
  twistMotion=Free
 }
 {
  bone1Name=spine_04
  bone2Name=spine_03
  swing1LimitDegrees=45
  swing1Motion=Free
  swing2LimitDegrees=45
  swing2Motion=Free
  twistLimitDegrees=45
  twistMotion=Free
 }
 {
  bone1Name=spine_03
  bone2Name=spine_02
  swing1LimitDegrees=45
  swing1Motion=Free
  swing2LimitDegrees=45
  swing2Motion=Free
  twistLimitDegrees=45
  twistMotion=Free
 }
 {
  bone1Name=dyn_hood_l_05
  bone2Name=dyn_hood_r_05
  swing1LimitDegrees=45
  swing1Motion=Free
  swing2LimitDegrees=45
  swing2Motion=Free
  twistLimitDegrees=45
  twistMotion=Free
 }
 {bone1Name=dyn_hood_l_02a, bone2Name=dyn_hood_l_02, swing1LimitDegrees=30, swing2LimitDegrees=20, twistLimitDegrees=0}
 {bone1Name=dyn_hood_r_02a, bone2Name=dyn_hood_r_02, swing1LimitDegrees=30, swing2LimitDegrees=20, twistLimitDegrees=0}
 {bone1Name=dyn_hood_l_04a, bone2Name=dyn_hood_l_04, swing1LimitDegrees=30, swing2LimitDegrees=20, twistLimitDegrees=0}
 {bone1Name=dyn_hood_r_04a, bone2Name=dyn_hood_r_04, swing1LimitDegrees=30, swing2LimitDegrees=20, twistLimitDegrees=0}
 {
  bone1Name=head
  bone2Name=spine_05
  swing1LimitDegrees=45
  swing1Motion=Free
  swing2LimitDegrees=45
  swing2Motion=Free
  twistLimitDegrees=45
  twistMotion=Free
 }
 {
  bone1Name=upperarm_l
  bone2Name=clavicle_l
  swing1LimitDegrees=45
  swing1Motion=Free
  swing2LimitDegrees=45
  swing2Motion=Free
  twistLimitDegrees=45
  twistMotion=Free
 }
 {
  bone1Name=upperarm_r
  bone2Name=clavicle_r
  swing1LimitDegrees=45
  swing1Motion=Free
  swing2LimitDegrees=45
  swing2Motion=Free
  twistLimitDegrees=45
  twistMotion=Free
 }
]
