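# Physics asset for secondary motion on the right lower leg of the Galileo Rocket female medium player body.
# Box bodies form knee-cloth flaps chained to the calf capsule, with swing1 and twist locked so each hinges on swing2.
# A boot pull and knot master hang off the calf twist body, the master through a free constraint.
# The tightly limited knot carries three lace capsules in parallel.
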PhysicsAsset /Game/Characters/Player/Female/Medium/Bodies/F_MED_Galileo_Rocket/Meshes/F_MED_Galileo_Rocket_Dyn_leg_r
bodies {
 calf_r [{shapeType=Capsule, length=8.25, radius=4.5625, center=<9.901 4.4974 -0.2485>, rotation=<rot 90 -14.0363 -34.6582>}]
 calf_twist_01_r [{shapeType=Capsule, length=1.375, radius=0.8125, rotation=<rot 90 0 0>}]
 dyn_boot_pull_r [{shapeType=Box, extentX=2.75, extentY=1.375, extentZ=0.5, center=<1.5 0 0>}]
 dyn_boot_string_knot_r [{shapeType=Capsule, length=0.5, radius=0.3, center=<-0.1 0 0>, rotation=<rot 90 0 0>}]
 dyn_boot_string_knot_r_MSTR [{shapeType=Capsule, length=0.2, radius=0.2}]
 dyn_boot_string_r_01 [{shapeType=Capsule, length=1, radius=0.3, center=<1 0 0>, rotation=<rot 90 0 0>}]
 dyn_boot_string_r_02 [{shapeType=Capsule, length=1, radius=0.3, center=<1 0 0>, rotation=<rot 90 0 0>}]
 dyn_boot_string_r_03 [{shapeType=Capsule, length=1, radius=0.3, center=<1 0 0>, rotation=<rot 90 0 0>}]
 dyn_kneecloth_bk_in_r_01 [{shapeType=Box, extentX=2.75, extentY=1.375, extentZ=1.0625, center=<1.4884 0 0>}]
 dyn_kneecloth_bk_in_r_02 [{shapeType=Box, extentX=2.75, extentY=1.375, extentZ=1.0625, center=<1.4884 0 0>}]
 dyn_kneecloth_bk_out_r_01 [{shapeType=Box, extentX=2.75, extentY=1.375, extentZ=0.75, center=<1.4884 0 0>}]
 dyn_kneecloth_bk_out_r_02 [{shapeType=Box, extentX=2.75, extentY=1.375, extentZ=0.75, center=<1.4884 0 0>}]
 dyn_kneecloth_fr_in_r_01 [{shapeType=Box, extentX=2.75, extentY=1.375, extentZ=0.5, center=<1.4884 0 0>}]
 dyn_kneecloth_fr_in_r_02 [{shapeType=Box, extentX=2.75, extentY=1.375, extentZ=0.5, center=<1.4884 0 0>}]
 dyn_kneecloth_fr_out_r_01 [{shapeType=Box, extentX=2.75, extentY=1.375, extentZ=1.125, center=<1.4884 0 0>}]
}
constraints [
 {
  bone1Name=dyn_kneecloth_bk_out_r_02
  bone2Name=dyn_kneecloth_bk_out_r_01
  swing1LimitDegrees=10
  swing1Motion=Locked
  swing2LimitDegrees=25
  twistLimitDegrees=2
  twistMotion=Locked
 }
 {
  bone1Name=dyn_kneecloth_fr_in_r_02
  bone2Name=dyn_kneecloth_fr_in_r_01
  swing1LimitDegrees=10
  swing1Motion=Locked
  swing2LimitDegrees=25
  twistLimitDegrees=2
  twistMotion=Locked
 }
 {
  bone1Name=dyn_kneecloth_bk_in_r_02
  bone2Name=dyn_kneecloth_bk_in_r_01
  swing1LimitDegrees=10
  swing1Motion=Locked
  swing2LimitDegrees=25
  twistLimitDegrees=2
  twistMotion=Locked
 }
 {
  bone1Name=dyn_kneecloth_fr_out_r_01
  bone2Name=calf_r
  swing1LimitDegrees=10
  swing1Motion=Locked
  swing2LimitDegrees=25
  twistLimitDegrees=2
  twistMotion=Locked
 }
 {
  bone1Name=dyn_kneecloth_bk_out_r_01
  bone2Name=calf_r
  swing1LimitDegrees=10
  swing1Motion=Locked
  swing2LimitDegrees=25
  twistLimitDegrees=2
  twistMotion=Locked
 }
 {
  bone1Name=dyn_kneecloth_fr_in_r_01
  bone2Name=calf_r
  swing1LimitDegrees=10
  swing1Motion=Locked
  swing2LimitDegrees=25
  twistLimitDegrees=2
  twistMotion=Locked
 }
 {
  bone1Name=dyn_kneecloth_bk_in_r_01
  bone2Name=calf_r
  swing1LimitDegrees=10
  swing1Motion=Locked
  swing2LimitDegrees=25
  twistLimitDegrees=2
  twistMotion=Locked
 }
 {bone1Name=dyn_boot_pull_r, bone2Name=calf_twist_01_r, swing1LimitDegrees=10, swing2LimitDegrees=45, twistLimitDegrees=2}
 {
  bone1Name=dyn_boot_string_knot_r_MSTR
  bone2Name=calf_twist_01_r
  swing1LimitDegrees=45
  swing1Motion=Free
  swing2LimitDegrees=45
  swing2Motion=Free
  twistLimitDegrees=45
  twistMotion=Free
 }
 {bone1Name=dyn_boot_string_knot_r, bone2Name=dyn_boot_string_knot_r_MSTR, swing1LimitDegrees=1, swing2LimitDegrees=1, twistLimitDegrees=1}
 {bone1Name=dyn_boot_string_r_01, bone2Name=dyn_boot_string_knot_r, swing1LimitDegrees=45, swing2LimitDegrees=45, twistLimitDegrees=5}
 {bone1Name=dyn_boot_string_r_02, bone2Name=dyn_boot_string_knot_r, swing1LimitDegrees=45, swing2LimitDegrees=45, twistLimitDegrees=5}
 {bone1Name=dyn_boot_string_r_03, bone2Name=dyn_boot_string_knot_r, swing1LimitDegrees=45, swing2LimitDegrees=45, twistLimitDegrees=5}
 {
  bone1Name=calf_twist_01_r
  bone2Name=calf_r
  swing1LimitDegrees=45
  swing1Motion=Free
  swing2LimitDegrees=45
  swing2Motion=Free
  twistLimitDegrees=45
  twistMotion=Free
 }
]
