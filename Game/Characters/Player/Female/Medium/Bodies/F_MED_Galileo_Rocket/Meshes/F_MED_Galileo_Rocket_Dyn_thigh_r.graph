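# Physics asset for the right-thigh dynamic piece of the F_MED_Galileo_Rocket female player body, simulating a dangling gun attachment on that bone chain.
# It has two capsule bodies: a long capsule on dyn_gun and a tiny one on the dyn_gun_MSTR anchor bone.
# A single constraint links them with tight swing and twist limits, so the gun can only wobble slightly.
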PhysicsAsset /Game/Characters/Player/Female/Medium/Bodies/F_MED_Galileo_Rocket/Meshes/F_MED_Galileo_Rocket_Dyn_thigh_r
bodies {
 dyn_gun [{shapeType=Capsule, length=16.893, radius=2.4301, center=<3.8198 -0.4823 0.6614>, rotation=<rot 0 -92.2267 -83.3031>}]
 dyn_gun_MSTR [{shapeType=Capsule, length=0.2, radius=0.2}]
}
constraints [{bone1Name=dyn_gun, bone2Name=dyn_gun_MSTR, swing1LimitDegrees=5, swing2LimitDegrees=5, twistLimitDegrees=2}]
