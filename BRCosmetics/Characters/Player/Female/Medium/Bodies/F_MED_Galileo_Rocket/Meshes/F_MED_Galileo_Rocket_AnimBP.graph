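# Post-process style AnimBP for the F_MED_Galileo_Rocket cosmetic body, adding secondary motion to hood, straps, waist and loincloth over the parent mesh's copied pose.
# Look At, constraints, alpha-driven bone offsets and AnimDynamics feed a cached pose, then backbling-selected hood RigidBody branches and a waist RigidBody.
# The event graph sets offset alphas from sprint, crouch and thigh rotation, and overrides gravity during skydiving.
# Leg RigidBody nodes are unconnected to the output.
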AnimBlueprint /BRCosmetics/Characters/Player/Female/Medium/Bodies/F_MED_Galileo_Rocket/Meshes/F_MED_Galileo_Rocket_AnimBP
target_skeleton=&/F_MED_Galileo_Rocket_Skeleton
graphs [
 AnimGraph {nodes [
   n1 = Get gravity_override() -> (gravity_override:Vector)
   n2 = RigidBody!(ComponentPose, Alpha, OverrideWorldGravity:Vector=@n1, ExternalForce) -> (Pose) {
    bClampLinearTranslationLimitToRefPose=true
    bOverrideWorldGravity=true
    overridePhysicsAsset=/Game/Characters/Player/Female/Medium/Bodies/F_MED_Galileo_Rocket/Meshes/F_MED_Galileo_Rocket_Dyn_leg_r
    simulationSpace=BaseBoneSpace
    alphaBoolBlend {blendOption=Linear}
    baseBoneRef {boneName=calf_r}
    componentAppliedLinearAccClamp=<3000 3000 3000>
    componentLinearAccScale=<0.4 0.3 0.3>
    componentLinearVelScale=<0.4 0.3 0.3>
    overrideWorldGravity {z=-980}
   }
   n3 = RigidBody!(ComponentPose=@n2, Alpha, ExternalForce) -> (Pose) {
    bClampLinearTranslationLimitToRefPose=true
    bOverrideWorldGravity=true
    overridePhysicsAsset=/Game/Characters/Player/Female/Medium/Bodies/F_MED_Galileo_Rocket/Meshes/F_MED_Galileo_Rocket_Dyn_thigh_r
    simulationSpace=BaseBoneSpace
    alphaBoolBlend {blendOption=Linear}
    baseBoneRef {boneName=thigh_r}
    componentAppliedLinearAccClamp=<3000 3000 3000>
    componentLinearAccScale=<0.4 0.3 0.3>
    componentLinearVelScale=<0.4 0.3 0.3>
    overrideWorldGravity {z=-980}
   }
   RigidBody!(ComponentPose=@n3, Alpha, OverrideWorldGravity:Vector=@n1, ExternalForce) -> (Pose) {
    bClampLinearTranslationLimitToRefPose=true
    bOverrideWorldGravity=true
    overridePhysicsAsset=/Game/Characters/Player/Female/Medium/Bodies/F_MED_Galileo_Rocket/Meshes/F_MED_Galileo_Rocket_Dyn_leg_l
    simulationSpace=BaseBoneSpace
    alphaBoolBlend {blendOption=Linear}
    baseBoneRef {boneName=calf_l}
    componentAppliedLinearAccClamp=<3000 3000 3000>
    componentLinearAccScale=<0.4 0.3 0.3>
    componentLinearVelScale=<0.4 0.3 0.3>
    overrideWorldGravity {z=-980}
   }
   n4 = RigidBody(ComponentPose=Local To Component(LocalPose=Use cached pose 'post Hood Pose'()), OverrideWorldGravity=@n1) {
    bClampLinearTranslationLimitToRefPose=true
    bOverrideWorldGravity=true
    overridePhysicsAsset=&/F_MED_Galileo_Rocket_Dyn_Waist
    simulationSpace=BaseBoneSpace
    alphaBoolBlend {blendOption=Linear}
    baseBoneRef {boneName=pelvis}
    componentAppliedLinearAccClamp=<3000 3000 3000>
    componentLinearAccScale=<0.32 0.3 0.3>
    componentLinearVelScale=<0.32 0.3 0.3>
    overrideWorldGravity {z=-980}
   }
   Output Pose(Result=Component To Local(ComponentPose=@n4)) {layerGroup=None}
   n5 = NOT Boolean(A=Get bRigidBodyEnabled())
   n6 = Look At - Bone: dyn_hood_l_MSTR!(ComponentPose=Local To Component(LocalPose=Copy Pose From Mesh() {bUseAttachedParent=true}), bAlphaBoolEnabled:bool=@n5) -> (Pose) {
    alphaInputType=Bool
    alphaBoolBlend {blendOption=Linear}
    boneToModify {boneName=dyn_hood_l_MSTR}
    lookAtLocation=<0 -3 -4>
    lookAtTarget {boneReference {boneName=spine_02}}
    lookAt_Axis {axis {y=0, z=-1}}
   }
   n7 = Look At - Bone: dyn_hood_r_MSTR!(ComponentPose=@n6, bAlphaBoolEnabled:bool=@n5) -> (Pose) {
    alphaInputType=Bool
    alphaBoolBlend {blendOption=Linear}
    boneToModify {boneName=dyn_hood_r_MSTR}
    lookAtLocation=<0 -3 4>
    lookAtTarget {boneReference {boneName=spine_02}}
    lookAt_Axis {axis {y=0, z=-1}}
   }
   n8 = "Constraint - dyn_waist_buckle (PPPPP)"!(
    ComponentPose=@n7
    Alpha=1
    ConstraintWeights_0=0.0084
    ConstraintWeights_1:float=0.032
    ConstraintWeights_2:float=0.0551
    ConstraintWeights_3:float=0.0104
    ConstraintWeights_4:float=0.8941
   ) -> (Pose) {
    alphaBoolBlend {blendOption=Linear}
    boneToModify {boneName=dyn_waist_buckle}
    constraintSetup [
     {targetBone {boneName=dyn_loincloth_fr_r_MSTR}}
     {targetBone {boneName=spine_01}}
     {targetBone {boneName=spine_02}}
     {targetBone {boneName=spine_04}}
     {targetBone {boneName=pelvis}}
    ]
    constraintWeights [1,1,1,1,1]
   }
   n9 = "Constraint - dyn_waist_belt_fr_r_MSTR (PPPPPPP)"!(
    ComponentPose=@n8
    Alpha=1
    ConstraintWeights_0=0.0547
    ConstraintWeights_1:float=0.0053
    ConstraintWeights_2:float=0.0085
    ConstraintWeights_3:float=0.0191
    ConstraintWeights_4:float=0.0011
    ConstraintWeights_5:float=0.0049
    ConstraintWeights_6:float=0.9064
   ) -> (Pose) {
    alphaBoolBlend {blendOption=Linear}
    boneToModify {boneName=dyn_waist_belt_fr_r_MSTR}
    constraintSetup [
     {targetBone {boneName=thigh_twist_01_r}}
     {targetBone {boneName=dyn_loincloth_fr_r_MSTR}}
     {targetBone {boneName=spine_01}}
     {targetBone {boneName=spine_02}}
     {targetBone {boneName=spine_03}}
     {targetBone {boneName=spine_04}}
     {targetBone {boneName=pelvis}}
    ]
    constraintWeights [1,1,1,1,1,1,1]
   }
   n10 = Get springing alpha() -> (springing alpha:double)
   n11 = "Transform (Modify) Bone - Bone: dyn_loincloth_fr_l_MSTR"(ComponentPose=@n9, Alpha=@n10, Translation="-1.030,-4.921,-0.162", Rotation="-0.000,-31.384,-47.772", Scale="1,1,1") {
    rotationMode=BMM_Additive
    rotationSpace=BCS_BoneSpace
    translationMode=BMM_Additive
    translationSpace=BCS_BoneSpace
    alphaBoolBlend {blendOption=Linear}
    alphaScaleBiasClamp {bInterpResult=true}
    boneToModify {boneName=dyn_loincloth_fr_l_MSTR}
    rotation {roll=-47.772, yaw=-31.384}
    translation=<-1.03 -4.921 -0.162>
   }
   n12 = "Transform (Modify) Bone - Bone: dyn_loincloth_fr_r_MSTR"(ComponentPose=@n11, Alpha=@n10, Translation="-6.930,6.071,-1.165", Rotation="-11.645,10.267,66.186", Scale="1,1,1") {
    rotationMode=BMM_Additive
    rotationSpace=BCS_BoneSpace
    translationMode=BMM_Additive
    translationSpace=BCS_BoneSpace
    alphaBoolBlend {blendOption=Linear}
    alphaScaleBiasClamp {bInterpResult=true}
    boneToModify {boneName=dyn_loincloth_fr_r_MSTR}
    rotation=<rot -11.645 10.267 66.186>
    translation=<-6.93 6.071 -1.165>
   }
   n13 = Get crouch alpha() -> (crouch alpha:double)
   n14 = "Transform (Modify) Bone - Bone: dyn_loincloth_fr_l_MSTR"(ComponentPose=@n12, Alpha=@n13, Translation="-5.896,-2.545,-2.249", Rotation="70.596,-34.839,-128.026", Scale="1,1,1") {
    rotationMode=BMM_Additive
    rotationSpace=BCS_BoneSpace
    translationMode=BMM_Additive
    translationSpace=BCS_BoneSpace
    alphaBoolBlend {blendOption=Linear}
    alphaScaleBiasClamp {bInterpResult=true}
    boneToModify {boneName=dyn_loincloth_fr_l_MSTR}
    rotation=<rot 70.5958 -34.8391 -128.0256>
    translation=<-5.896 -2.5447 -2.249>
   }
   n15 = "Transform (Modify) Bone - Bone: dyn_waist_buckle"(ComponentPose=@n14, Alpha=@n13, Translation="0.543,-1.088,0.636", Rotation="-10.644,12.906,-10.488", Scale="1,1,1") {
    rotationMode=BMM_Additive
    rotationSpace=BCS_BoneSpace
    translationMode=BMM_Additive
    translationSpace=BCS_BoneSpace
    alphaBoolBlend {blendOption=Linear}
    alphaScaleBiasClamp {bInterpResult=true}
    boneToModify {boneName=dyn_waist_buckle}
    rotation=<rot -10.644 12.9063 -10.4882>
    translation=<0.543 -1.088 0.636>
   }
   n16 = "Transform (Modify) Bone - Bone: dyn_loincloth_fr_r_MSTR"(ComponentPose=@n15, Alpha=Get thigh rotation_r() -> (thigh rotation_r:double), Translation="-5.112,0,0", Rotation="0,0,30", Scale="1,1,1") {
    rotationMode=BMM_Additive
    rotationSpace=BCS_BoneSpace
    translationMode=BMM_Additive
    translationSpace=BCS_BoneSpace
    alphaBoolBlend {blendOption=Linear}
    alphaScaleBiasClamp {bInterpResult=true}
    boneToModify {boneName=dyn_loincloth_fr_r_MSTR}
    rotation {roll=30}
    translation {x=-5.112}
   }
   n17 = Anim Dynamics - Bone: dyn_armstrap_waist_r_02(ComponentPose=@n16, Alpha=1, ExternalForce="0,0,0") {
    angularDampingOverride=1
    angularSpringConstant=2500
    bAngularSpring=true
    bGravityOverrideInSimSpace=true
    bOverrideAngularDamping=true
    bUseGravityOverride=true
    gravityScale=0.35
    simulationSpace=BoneRelative
    alphaBoolBlend {blendOption=Linear}
    boundBone {boneName=dyn_armstrap_waist_r_02}
    componentAppliedLinearAccClamp=<3000 3000 3000>
    componentLinearAccScale=<0.2 0.1 0.1>
    componentLinearVelScale=<0.2 0.1 0.1>
    gravityOverride {z=-200}
    physicsBodyDefinitions [{
      sphereCollisionRadius=0
      boundBone {boneName=dyn_armstrap_waist_r_02}
      boxExtents=<5 10 10>
      constraintSetup {angularLimitsMax=<0 0 10>, angularLimitsMin=<0 -10 -10>, angularTarget=<1 0 0>}
      localJointOffset=<2 0 0>
     }]
    relativeSpaceBone {boneName=upperarm_r}
   }
   n18 = Anim Dynamics - Bone: dyn_armstrap_waist_r_01(ComponentPose=@n17, Alpha=1, ExternalForce="0,0,0") {
    angularDampingOverride=1
    angularSpringConstant=2500
    bAngularSpring=true
    bGravityOverrideInSimSpace=true
    bOverrideAngularDamping=true
    bUseGravityOverride=true
    gravityScale=0.35
    simulationSpace=BoneRelative
    alphaBoolBlend {blendOption=Linear}
    boundBone {boneName=dyn_armstrap_waist_r_01}
    componentAppliedLinearAccClamp=<3000 3000 3000>
    componentLinearAccScale=<0.2 0.1 0.1>
    componentLinearVelScale=<0.2 0.1 0.1>
    gravityOverride {z=-200}
    physicsBodyDefinitions [{
      sphereCollisionRadius=0
      boundBone {boneName=dyn_armstrap_waist_r_01}
      boxExtents=<5 10 10>
      constraintSetup {angularLimitsMax=<0 0 10>, angularLimitsMin=<0 -10 -10>, angularTarget=<1 0 0>}
      localJointOffset=<2 0 0>
     }]
    relativeSpaceBone {boneName=upperarm_r}
   }
   n19 = Anim Dynamics - Bone: dyn_beltstrap_waist_bk_01(ComponentPose=@n18, Alpha=1, ExternalForce="0,0,0") {
    angularDampingOverride=1
    angularSpringConstant=2500
    bAngularSpring=true
    bGravityOverrideInSimSpace=true
    bOverrideAngularDamping=true
    bUseGravityOverride=true
    gravityScale=0.35
    simulationSpace=BoneRelative
    alphaBoolBlend {blendOption=Linear}
    boundBone {boneName=dyn_beltstrap_waist_bk_01}
    componentAppliedLinearAccClamp=<3000 3000 3000>
    componentLinearAccScale=<0.2 0.1 0.1>
    componentLinearVelScale=<0.2 0.1 0.1>
    gravityOverride {z=-200}
    physicsBodyDefinitions [{
      sphereCollisionRadius=0
      boundBone {boneName=dyn_beltstrap_waist_bk_01}
      boxExtents=<5 10 10>
      constraintSetup {angularLimitsMax=<0 0 10>, angularLimitsMin=<0 -10 -10>, angularTarget=<1 0 0>}
      localJointOffset=<2 0 0>
     }]
    relativeSpaceBone {boneName=pelvis}
   }
   n20 = Anim Dynamics - Bone: dyn_beltstrap_waist_bk_02(ComponentPose=@n19, Alpha=1, ExternalForce="0,0,0") {
    angularDampingOverride=1
    angularSpringConstant=2500
    bAngularSpring=true
    bGravityOverrideInSimSpace=true
    bOverrideAngularDamping=true
    bUseGravityOverride=true
    gravityScale=0.35
    simulationSpace=BoneRelative
    alphaBoolBlend {blendOption=Linear}
    boundBone {boneName=dyn_beltstrap_waist_bk_02}
    componentAppliedLinearAccClamp=<3000 3000 3000>
    componentLinearAccScale=<0.2 0.1 0.1>
    componentLinearVelScale=<0.2 0.1 0.1>
    gravityOverride {z=-200}
    physicsBodyDefinitions [{
      sphereCollisionRadius=0
      boundBone {boneName=dyn_beltstrap_waist_bk_02}
      boxExtents=<5 10 10>
      constraintSetup {angularLimitsMax=<0 0 10>, angularLimitsMin=<0 -10 -10>, angularTarget=<1 0 0>}
      localJointOffset=<2 0 0>
     }]
    relativeSpaceBone {boneName=pelvis}
   }
   n21 = Anim Dynamics - Bone: dyn_beltstrap_hips_bk_01(ComponentPose=@n20, Alpha=1, ExternalForce="0,0,0") {
    angularDampingOverride=1
    angularSpringConstant=2500
    bAngularSpring=true
    bGravityOverrideInSimSpace=true
    bOverrideAngularDamping=true
    bUseGravityOverride=true
    gravityScale=0.35
    simulationSpace=BoneRelative
    alphaBoolBlend {blendOption=Linear}
    boundBone {boneName=dyn_beltstrap_hips_bk_01}
    componentAppliedLinearAccClamp=<3000 3000 3000>
    componentLinearAccScale=<0.2 0.1 0.1>
    componentLinearVelScale=<0.2 0.1 0.1>
    gravityOverride {z=-200}
    physicsBodyDefinitions [{
      sphereCollisionRadius=0
      boundBone {boneName=dyn_beltstrap_hips_bk_01}
      boxExtents=<5 10 10>
      constraintSetup {angularLimitsMax=<0 0 10>, angularLimitsMin=<0 -10 -10>, angularTarget=<1 0 0>}
      localJointOffset=<2 0 0>
     }]
    relativeSpaceBone {boneName=pelvis}
   }
   n22 = Anim Dynamics - Bone: dyn_beltstrap_hips_bk_02(ComponentPose=@n21, Alpha=1, ExternalForce="0,0,0") {
    angularDampingOverride=1
    angularSpringConstant=2500
    bAngularSpring=true
    bGravityOverrideInSimSpace=true
    bOverrideAngularDamping=true
    bUseGravityOverride=true
    gravityScale=0.35
    simulationSpace=BoneRelative
    alphaBoolBlend {blendOption=Linear}
    boundBone {boneName=dyn_beltstrap_hips_bk_02}
    componentAppliedLinearAccClamp=<3000 3000 3000>
    componentLinearAccScale=<0.2 0.1 0.1>
    componentLinearVelScale=<0.2 0.1 0.1>
    gravityOverride {z=-200}
    physicsBodyDefinitions [{
      sphereCollisionRadius=0
      boundBone {boneName=dyn_beltstrap_hips_bk_02}
      boxExtents=<5 10 10>
      constraintSetup {angularLimitsMax=<0 0 10>, angularLimitsMin=<0 -10 -10>, angularTarget=<1 0 0>}
      localJointOffset=<2 0 0>
     }]
    relativeSpaceBone {boneName=pelvis}
   }
   Save cached pose 'initial Pose'(Pose=Component To Local(ComponentPose=@n22)) {cacheName=initial Pose}
   n23 = Get gravity_override()
   n24 = RigidBody(ComponentPose=Local To Component(LocalPose=Use cached pose 'initial Pose'()), OverrideWorldGravity=@n23) {
    bClampLinearTranslationLimitToRefPose=true
    bOverrideWorldGravity=true
    overridePhysicsAsset=&/F_MED_Galileo_Rocket_Dyn_hood_backbling
    simulationSpace=BaseBoneSpace
    alphaBoolBlend {blendOption=Linear}
    baseBoneRef {boneName=spine_04}
    componentAppliedLinearAccClamp=<3000 3000 3000>
    componentLinearAccScale=<0.4 0.3 0.3>
    componentLinearVelScale=<0.4 0.3 0.3>
    overrideWorldGravity {z=-980}
   }
   n25 = "Transform (Modify) Bone - Bone: dyn_hood_l_02a"!(ComponentPose=@n24, Alpha=1, Translation="0,0,0", Rotation="0,0,0") -> (Pose) {
    rotationMode=BMM_Additive
    rotationSpace=BCS_BoneSpace
    scaleMode=BMM_Additive
    scaleSpace=BCS_BoneSpace
    translationMode=BMM_Additive
    translationSpace=BCS_BoneSpace
    alphaBoolBlend {blendOption=Linear}
    alphaScaleBiasClamp {bInterpResult=true}
    boneToModify {boneName=dyn_hood_l_02a}
    scale {z=0.2}
   }
   n26 = "Transform (Modify) Bone - Bone: dyn_hood_l_03"!(ComponentPose=@n25, Alpha=1, Translation="0,0,0", Rotation="0,0,0") -> (Pose) {
    rotationMode=BMM_Additive
    rotationSpace=BCS_BoneSpace
    scaleMode=BMM_Additive
    scaleSpace=BCS_BoneSpace
    translationMode=BMM_Additive
    translationSpace=BCS_BoneSpace
    alphaBoolBlend {blendOption=Linear}
    alphaScaleBiasClamp {bInterpResult=true}
    boneToModify {boneName=dyn_hood_l_03}
    scale {z=0.2}
   }
   n27 = "Transform (Modify) Bone - Bone: dyn_hood_l_04"!(ComponentPose=@n26, Alpha=1, Translation="0,0,0", Rotation="0,0,0") -> (Pose) {
    rotationMode=BMM_Additive
    rotationSpace=BCS_BoneSpace
    scaleMode=BMM_Additive
    scaleSpace=BCS_BoneSpace
    translationMode=BMM_Additive
    translationSpace=BCS_BoneSpace
    alphaBoolBlend {blendOption=Linear}
    alphaScaleBiasClamp {bInterpResult=true}
    boneToModify {boneName=dyn_hood_l_04}
    scale {z=0.2}
   }
   n28 = "Transform (Modify) Bone - Bone: dyn_hood_l_05"!(ComponentPose=@n27, Alpha=1, Translation="0,0,0", Rotation="0,0,0") -> (Pose) {
    rotationMode=BMM_Additive
    rotationSpace=BCS_BoneSpace
    scaleMode=BMM_Additive
    scaleSpace=BCS_BoneSpace
    translationMode=BMM_Additive
    translationSpace=BCS_BoneSpace
    alphaBoolBlend {blendOption=Linear}
    alphaScaleBiasClamp {bInterpResult=true}
    boneToModify {boneName=dyn_hood_l_05}
    scale {z=0.2}
   }
   n29 = "Transform (Modify) Bone - Bone: dyn_hood_l_04a"!(ComponentPose=@n28, Alpha=1, Translation="0,0,0", Rotation="0,0,0") -> (Pose) {
    rotationMode=BMM_Additive
    rotationSpace=BCS_BoneSpace
    scaleMode=BMM_Additive
    scaleSpace=BCS_BoneSpace
    translationMode=BMM_Additive
    translationSpace=BCS_BoneSpace
    alphaBoolBlend {blendOption=Linear}
    alphaScaleBiasClamp {bInterpResult=true}
    boneToModify {boneName=dyn_hood_l_04a}
    scale {x=0.2}
   }
   n30 = "Transform (Modify) Bone - Bone: dyn_hood_r_02a"!(ComponentPose=@n29, Alpha=1, Translation="0,0,0", Rotation="0,0,0") -> (Pose) {
    rotationMode=BMM_Additive
    rotationSpace=BCS_BoneSpace
    scaleMode=BMM_Additive
    scaleSpace=BCS_BoneSpace
    translationMode=BMM_Additive
    translationSpace=BCS_BoneSpace
    alphaBoolBlend {blendOption=Linear}
    alphaScaleBiasClamp {bInterpResult=true}
    boneToModify {boneName=dyn_hood_r_02a}
    scale {z=0.2}
   }
   n31 = "Transform (Modify) Bone - Bone: dyn_hood_r_03"!(ComponentPose=@n30, Alpha=1, Translation="0,0,0", Rotation="0,0,0") -> (Pose) {
    rotationMode=BMM_Additive
    rotationSpace=BCS_BoneSpace
    scaleMode=BMM_Additive
    scaleSpace=BCS_BoneSpace
    translationMode=BMM_Additive
    translationSpace=BCS_BoneSpace
    alphaBoolBlend {blendOption=Linear}
    alphaScaleBiasClamp {bInterpResult=true}
    boneToModify {boneName=dyn_hood_r_03}
    scale {z=0.2}
   }
   n32 = "Transform (Modify) Bone - Bone: dyn_hood_r_04"!(ComponentPose=@n31, Alpha=1, Translation="0,0,0", Rotation="0,0,0") -> (Pose) {
    rotationMode=BMM_Additive
    rotationSpace=BCS_BoneSpace
    scaleMode=BMM_Additive
    scaleSpace=BCS_BoneSpace
    translationMode=BMM_Additive
    translationSpace=BCS_BoneSpace
    alphaBoolBlend {blendOption=Linear}
    alphaScaleBiasClamp {bInterpResult=true}
    boneToModify {boneName=dyn_hood_r_04}
    scale {z=0.2}
   }
   n33 = "Transform (Modify) Bone - Bone: dyn_hood_r_05"!(ComponentPose=@n32, Alpha=1, Translation="0,0,0", Rotation="0,0,0") -> (Pose) {
    rotationMode=BMM_Additive
    rotationSpace=BCS_BoneSpace
    scaleMode=BMM_Additive
    scaleSpace=BCS_BoneSpace
    translationMode=BMM_Additive
    translationSpace=BCS_BoneSpace
    alphaBoolBlend {blendOption=Linear}
    alphaScaleBiasClamp {bInterpResult=true}
    boneToModify {boneName=dyn_hood_r_05}
    scale {z=0.2}
   }
   n34 = "Transform (Modify) Bone - Bone: dyn_hood_r_04a"!(ComponentPose=@n33, Alpha=1, Translation="0,0,0", Rotation="0,0,0") -> (Pose) {
    rotationMode=BMM_Additive
    rotationSpace=BCS_BoneSpace
    scaleMode=BMM_Additive
    scaleSpace=BCS_BoneSpace
    translationMode=BMM_Additive
    translationSpace=BCS_BoneSpace
    alphaBoolBlend {blendOption=Linear}
    alphaScaleBiasClamp {bInterpResult=true}
    boneToModify {boneName=dyn_hood_r_04a}
    scale {x=0.2}
   }
   n35 = RigidBody(ComponentPose=Local To Component(LocalPose=Use cached pose 'initial Pose'()), OverrideWorldGravity=@n23) {
    bClampLinearTranslationLimitToRefPose=true
    bOverrideWorldGravity=true
    overridePhysicsAsset=&/F_MED_Galileo_Rocket_Dyn_hood
    simulationSpace=BaseBoneSpace
    alphaBoolBlend {blendOption=Linear}
    baseBoneRef {boneName=spine_04}
    componentAppliedLinearAccClamp=<3000 3000 3000>
    componentLinearAccScale=<0.4 0.3 0.3>
    componentLinearVelScale=<0.4 0.3 0.3>
    overrideWorldGravity {z=-980}
   }
   Save cached pose 'post Hood Pose'(Pose=Blend Poses by bool(
     BlendPose_0=Component To Local(ComponentPose=@n34)
     BlendPose_1=Component To Local(ComponentPose=@n35)
     bActiveValue=Get isWearingBackbling()
    ) {blendType=Linear}) {cacheName=post Hood Pose}
  ]}
 EventGraph {nodes [
   Event BlueprintUpdateAnimation()
   > n1 = Sequence!(execute) -> (then_0, then_1, then_2:Exec)
   > n2 = Branch(Condition=Get bIsSkyDiving())
   > n3 = Branch(Condition=Get bIsParachuteOpened())
   Set gravity_override(execute=[@n2.else, @n3.then], gravity_override="0, 0,-980")
   n4 = Branch(execute=@n3.else, Condition="float > float"(A=Get PawnVelocityZ(), B=50) {is=title K2Node_CallFunction})
   > Set gravity_override(gravity_override="0, 0,800")
   Set gravity_override(execute=@n4.else, gravity_override="0, 0,-100")
   n5 = Branch(execute=@n1.then_1, Condition=Get bIsSprinting())
   > Set thigh rotation_r(thigh rotation_r:double=0) -> (Output_Get:double)
   > Set thigh_rotation_l(thigh_rotation_l:double=0) -> (Output_Get:double)
   n6 = BreakTransform(InTransform=GetDeltaTransformFromRefPose(self=GetOwnerSkeletalMeshComponent(), BoneName=thigh_r, BaseName=pelvis))
   Set thigh rotation_r(execute=@n5.else, thigh rotation_r=MapRangeClamped(Value=BreakRotator(InRot=@n6.Rotation).Yaw, InRangeA=-50, InRangeB=-90))
   n7 = BreakTransform(InTransform=GetDeltaTransformFromRefPose(self=GetOwnerSkeletalMeshComponent(), BoneName=thigh_l, BaseName=pelvis))
   > Set thigh_rotation_l(thigh_rotation_l=MapRangeClamped(Value=BreakRotator(InRot=@n7.Rotation).Yaw, InRangeA=-50, InRangeB=-90))
   n8 = Branch(execute=@n1.then_2, Condition=OR Boolean(A=Get bIsSprinting(), B=Get bIsCrouchSprinting()))
   > Set springing alpha(springing alpha:double=1) -> (Output_Get:double)
   > Set crouch alpha(crouch alpha:double=0) -> (Output_Get:double)
   n9 = Branch(execute=@n8.else, Condition=Get bIsCrouching())
   > Set springing alpha(springing alpha=0)
   > Set crouch alpha(crouch alpha=1)
   Set crouch alpha(execute=@n9.else, crouch alpha=0)
   > Set springing alpha(springing alpha=0)
   Event OnCharacterCustomizationLoaded()
   n10 = CheckForBackbling()
   > Set isWearingBackbling(isWearingBackbling=@n10)
   > Set isWearingCape(isWearingCape=@n10.bIsBackblingLong)
  ]}
]
variables [
 crouch alpha {type=number}
 gravity_override {type=Vector}
 isWearingBackbling {type=boolean}
 isWearingCape {type=boolean}
 springing alpha {type=number}
 thigh rotation_r {type=number}
 thigh_rotation_l {type=number}
]
